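; GrayBox sliding double-door wall Blueprint for Creative that adds a dust effect around the door.
; Both begin and end overlap on the vol_Dust volume activate the NS_DoorDust component without a reset, with no filter on the overlapping actor.
; When the door is destroyed, it destroys the dust component.
; The construction script is empty, and no sliding logic appears in this graph.
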(event EventDestroyed
  (Components|DestroyComponent (Variables|Default|GetNS_DoorDust)))

(event OnComponentBeginOverlap(vol_Dust)
  (params (OverlappedComponent :type Primitive%20Component%20Object%20Reference) (OtherActor :type Actor%20Object%20Reference) (OtherComp :type Primitive%20Component%20Object%20Reference) (OtherBodyIndex :type Integer) (bFromSweep :type Boolean) (SweepResult :type Hit%20Result%20Structure%20%28by%20ref%29))
  (Components|Activation|Activate (Variables|Default|GetNS_DoorDust) false))

(event OnComponentEndOverlap(vol_Dust)
  (params (OverlappedComponent :type Primitive%20Component%20Object%20Reference) (OtherActor :type Actor%20Object%20Reference) (OtherComp :type Primitive%20Component%20Object%20Reference) (OtherBodyIndex :type Integer))
  (Components|Activation|Activate (Variables|Default|GetNS_DoorDust) false))

(fn ConstructionScript)
